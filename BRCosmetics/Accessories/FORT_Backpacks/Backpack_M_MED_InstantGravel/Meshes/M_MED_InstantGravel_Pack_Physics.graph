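# Physics asset for the InstantGravel backpack cosmetic, a creature-shaped pack with a head, long jointed ears, paws and legs.
# Capsule bodies cover each bone, with an added sphere on the head and zero-size capsules at the ear roots.
# Four-bone ear chains hang from the head, four-bone arm chains from spine_04, and two-bone legs from the pelvis.
# These limbs have tight swing and twist limits, while pelvis-to-spine and spine-to-head joints rotate freely.
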PhysicsAsset /BRCosmetics/Accessories/FORT_Backpacks/Backpack_M_MED_InstantGravel/Meshes/M_MED_InstantGravel_Pack_Physics
bodies {
 calf_l [{
   shapeType=Capsule
   length=3.7233
   radius=2.1293
   shapeName=calf_l_capsule
   center=<4.274 -0.4951 -1.1234>
   rotation=<rot -72.5225 -73.7391 76.6959>
  }]
 calf_r [{
   shapeType=Capsule
   length=3.7233
   radius=2.1293
   shapeName=calf_r_capsule
   center=<-4.0562 -0.5486 0.6564>
   rotation=<rot -71.1908 -74.2424 81.9912>
  }]
 ear_l_01 [{shapeType=Capsule, length=0, radius=0, shapeName=ear_l_01_capsule}]
 ear_l_02 [{shapeType=Capsule, length=1.5, radius=1, shapeName=ear_l_02_capsule, center=<1 0 0>}]
 ear_l_03 [{shapeType=Capsule, length=1.5, radius=1, shapeName=ear_l_03_capsule, center=<1 0 0>, rotation=<rot 90 0 0>}]
 ear_l_04 [{shapeType=Capsule, length=1.5, radius=1, shapeName=ear_l_04_capsule, center=<1 0 0>, rotation=<rot 90 0 0>}]
 ear_r_01 [{shapeType=Capsule, length=0, radius=0, shapeName=ear_r_01_capsule}]
 ear_r_02 [{shapeType=Capsule, length=1.5, radius=1, shapeName=ear_r_02_capsule, center=<-1 0 0>}]
 ear_r_03 [{shapeType=Capsule, length=1.5, radius=1, shapeName=ear_r_03_capsule, center=<-1 0 0>, rotation=<rot 90 0 0>}]
 ear_r_04 [{shapeType=Capsule, length=1.5, radius=1, shapeName=ear_r_04_capsule, center=<-1 0 0>, rotation=<rot 90 0 0>}]
 head [
  {shapeType=Sphere, radius=5.1402, shapeName=head_sphere, center=<3.4232 2.4146 0.1844>}
  {shapeType=Capsule, length=8.9197, radius=2.9489, shapeName=head_capsule, center=<1.3291 1.8543 0>, rotation=<rot 0 0.0002 -90>}
 ]
 paw_forearm_l [{shapeType=Capsule, length=2, radius=2.5, shapeName=paw_forearm_l_capsule, center=<1.5 0 0>, rotation=<rot 90 0 0>}]
 paw_forearm_r [{shapeType=Capsule, length=2, radius=2.5, shapeName=paw_forearm_r_capsule, center=<-1.5 0 0>, rotation=<rot 90 0 0>}]
 paw_l [{shapeType=Capsule, length=1.5, radius=2, shapeName=paw_l_capsule, center=<1.5 0 0>, rotation=<rot 90 0 0>}]
 paw_r [{shapeType=Capsule, length=1.5, radius=2, shapeName=paw_r_capsule, center=<-1.5 0 0>, rotation=<rot 90 0 0>}]
 paw_upperarm_l [{shapeType=Capsule, length=3, radius=2.5, shapeName=paw_upperarm_l_capsule, center=<1.5 0 0>, rotation=<rot 90 0 0>}]
 paw_upperarm_r [{shapeType=Capsule, length=3, radius=2.5, shapeName=paw_upperarm_r_capsule, center=<-1.5 0 0>, rotation=<rot 90 0 0>}]
 paw_wrist_l [{shapeType=Capsule, length=1.5, radius=2, shapeName=paw_wrist_l_capsule, center=<1.5 0 0>, rotation=<rot 90 0 0>}]
 paw_wrist_r [{shapeType=Capsule, length=1.5, radius=2, shapeName=paw_wrist_r_capsule, center=<-1.5 0 0>, rotation=<rot 90 0 0>}]
 pelvis [{shapeType=Capsule, length=5.0198, radius=5, shapeName=pelvis_capsule}]
 spine_04 [{shapeType=Capsule, length=3.0972, radius=5, shapeName=spine_04_capsule}]
 thight_l [{shapeType=Capsule, length=4.2261, radius=3.697, shapeName=thight_l_capsule, center=<3.699 0 -0.9774>, rotation=<rot 90 0 0>}]
 thight_r [{shapeType=Capsule, length=4.5452, radius=3.3779, shapeName=thight_r_capsule, center=<-4.0274 0 0.8282>, rotation=<rot 90 0 0>}]
}
constraints [
 {bone1Name=ear_l_02, bone2Name=ear_l_01, swing1LimitDegrees=30, swing2LimitDegrees=30, twistLimitDegrees=5}
 {bone1Name=ear_l_03, bone2Name=ear_l_02, swing1LimitDegrees=30, swing2LimitDegrees=30, twistLimitDegrees=5}
 {bone1Name=ear_l_04, bone2Name=ear_l_03, swing1LimitDegrees=30, swing2LimitDegrees=30, twistLimitDegrees=5}
 {bone1Name=ear_r_02, bone2Name=ear_r_01, swing1LimitDegrees=30, swing2LimitDegrees=30, twistLimitDegrees=5}
 {bone1Name=ear_r_03, bone2Name=ear_r_02, swing1LimitDegrees=30, swing2LimitDegrees=30, twistLimitDegrees=5}
 {bone1Name=ear_r_04, bone2Name=ear_r_03, swing1LimitDegrees=30, swing2LimitDegrees=30, twistLimitDegrees=5}
 {bone1Name=paw_forearm_l, bone2Name=paw_upperarm_l, swing1LimitDegrees=20, swing2LimitDegrees=20, twistLimitDegrees=2.5}
 {bone1Name=paw_forearm_r, bone2Name=paw_upperarm_r, swing1LimitDegrees=20, swing2LimitDegrees=40, twistLimitDegrees=4}
 {bone1Name=paw_wrist_l, bone2Name=paw_forearm_l, swing1LimitDegrees=7, swing2LimitDegrees=20, twistLimitDegrees=5}
 {bone1Name=paw_l, bone2Name=paw_wrist_l, swing1LimitDegrees=10, swing2LimitDegrees=25, twistLimitDegrees=2}
 {bone1Name=paw_wrist_r, bone2Name=paw_forearm_r, swing1LimitDegrees=7, swing2LimitDegrees=20, twistLimitDegrees=5}
 {bone1Name=paw_r, bone2Name=paw_wrist_r, swing1LimitDegrees=10, swing2LimitDegrees=25, twistLimitDegrees=2}
 {bone1Name=calf_l, bone2Name=thight_l, swing1LimitDegrees=10, swing2LimitDegrees=15, twistLimitDegrees=2}
 {bone1Name=calf_r, bone2Name=thight_r, swing1LimitDegrees=10, swing2LimitDegrees=15, twistLimitDegrees=2}
 {bone1Name=ear_l_01, bone2Name=head, swing1LimitDegrees=30, swing2LimitDegrees=30, twistLimitDegrees=5}
 {bone1Name=ear_r_01, bone2Name=head, swing1LimitDegrees=30, swing2LimitDegrees=30, twistLimitDegrees=5}
 {bone1Name=paw_upperarm_l, bone2Name=spine_04, swing1LimitDegrees=30, swing2LimitDegrees=25, twistLimitDegrees=5}
 {bone1Name=paw_upperarm_r, bone2Name=spine_04, swing1LimitDegrees=30, swing2LimitDegrees=25, twistLimitDegrees=5}
 {
  bone1Name=spine_04
  bone2Name=pelvis
  swing1LimitDegrees=45
  swing1Motion=Free
  swing2LimitDegrees=45
  swing2Motion=Free
  twistLimitDegrees=45
  twistMotion=Free
 }
 {bone1Name=thight_l, bone2Name=pelvis, swing1LimitDegrees=10, swing2LimitDegrees=30, twistLimitDegrees=2}
 {bone1Name=thight_r, bone2Name=pelvis, swing1LimitDegrees=10, swing2LimitDegrees=30, twistLimitDegrees=2}
 {
  bone1Name=head
  bone2Name=spine_04
  swing1LimitDegrees=45
  swing1Motion=Free
  swing2LimitDegrees=45
  swing2Motion=Free
  twistLimitDegrees=45
  twistMotion=Free
 }
]
