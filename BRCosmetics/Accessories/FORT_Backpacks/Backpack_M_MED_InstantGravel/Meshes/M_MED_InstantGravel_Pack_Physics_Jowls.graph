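# Physics asset for the jowls of the InstantGravel backpack cosmetic, defining bodies for secondary motion on the face.
# Small spheres on the left and right outer-cheek and lip-corner bones are each constrained directly to the head body, which combines a sphere and a capsule.
# The constraints set narrow swing limits and very tight twist limits, with lip corners given asymmetric swing and cheeks symmetric.
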PhysicsAsset /BRCosmetics/Accessories/FORT_Backpacks/Backpack_M_MED_InstantGravel/Meshes/M_MED_InstantGravel_Pack_Physics_Jowls
bodies {
 L_cheek_outer [{shapeType=Sphere, radius=1, shapeName=L_cheek_outer_sphere, center=<0.2366 0.6402 0.4352>}]
 L_lip_corner [{shapeType=Sphere, radius=1, shapeName=L_lip_corner_sphere, center=<-0.2757 0.9151 -0.1339>}]
 R_cheek_outer [{shapeType=Sphere, radius=1, shapeName=R_cheek_outer_sphere, center=<-0.1953 -0.5286 -0.3593>}]
 R_lip_corner [{shapeType=Sphere, radius=1, shapeName=R_lip_corner_sphere, center=<0.2757 -0.9151 0.1339>}]
 head [
  {shapeType=Sphere, radius=5.1402, shapeName=head_sphere, center=<3.4232 2.4146 0.1844>}
  {shapeType=Capsule, length=7.8773, radius=2.6388, shapeName=head_capsule, center=<1.8673 4.0133 0>, rotation=<rot 0 0.0001 -90>}
 ]
}
constraints [
 {bone1Name=R_lip_corner, bone2Name=head, swing1LimitDegrees=10, swing2LimitDegrees=20, twistLimitDegrees=2}
 {bone1Name=L_lip_corner, bone2Name=head, swing1LimitDegrees=10, swing2LimitDegrees=20, twistLimitDegrees=2}
 {bone1Name=L_cheek_outer, bone2Name=head, swing1LimitDegrees=15, swing2LimitDegrees=15, twistLimitDegrees=2}
 {bone1Name=R_cheek_outer, bone2Name=head, swing1LimitDegrees=15, swing2LimitDegrees=15, twistLimitDegrees=2}
]
