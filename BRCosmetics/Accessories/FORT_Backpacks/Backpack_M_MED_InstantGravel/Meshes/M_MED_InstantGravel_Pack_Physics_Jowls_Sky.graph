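# Physics asset for the InstantGravel backpack cosmetic (Sky variant), giving the face's jowls, lips, cheeks and tongue bodies for physical secondary motion.
# A two-shape head body anchors small spheres on the cheeks and lip corners through tight swing and twist limits.
# The tongue is a four-body chain from the head with only a few degrees of play per joint.
# Upper and lower middle lip capsules attach to the head with free angular motion.
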PhysicsAsset /BRCosmetics/Accessories/FORT_Backpacks/Backpack_M_MED_InstantGravel/Meshes/M_MED_InstantGravel_Pack_Physics_Jowls_Sky
bodies {
 C_lip_lower_mid [{
   shapeType=Capsule
   length=2.0107
   radius=1.3499
   shapeName=C_lip_lower_mid_capsule
   center=<0.0506 -0.2646 -1.2991>
   rotation=<rot 0.0002 -86.5995 -91.2974>
  }]
 C_lip_upper_mid [{
   shapeType=Capsule
   length=2.587
   radius=0.8892
   shapeName=C_lip_upper_mid_capsule
   center=<0.0686 0.0344 0.6493>
   rotation=<rot 0 84.4032 -85.3459>
  }]
 L_cheek_outer [{shapeType=Sphere, radius=1, shapeName=L_cheek_outer_sphere, center=<0.2366 0.6402 0.4352>}]
 L_lip_corner [{shapeType=Sphere, radius=1, shapeName=L_lip_corner_sphere, center=<-0.2757 0.9151 -0.1339>}]
 R_cheek_outer [{shapeType=Sphere, radius=1, shapeName=R_cheek_outer_sphere, center=<-0.1953 -0.5286 -0.3593>}]
 R_lip_corner [{shapeType=Sphere, radius=1, shapeName=R_lip_corner_sphere, center=<0.2757 -0.9151 0.1339>}]
 head [
  {shapeType=Sphere, radius=5.1402, shapeName=head_sphere, center=<3.4232 2.4146 0.1844>}
  {shapeType=Capsule, length=7.8773, radius=2.6388, shapeName=head_capsule, center=<1.8673 4.0133 0>, rotation=<rot 0 0.0001 -90>}
 ]
 tongue [{shapeType=Sphere, radius=1, shapeName=tongue_sphere, center=<0 2 0>}]
 tongue_02 [{shapeType=Sphere, radius=1, shapeName=tongue_02_sphere, center=<0 2 0>}]
 tongue_03 [{shapeType=Sphere, radius=1, shapeName=tongue_03_sphere, center=<0 0.5 0>}]
 tongue_04 [{shapeType=Sphere, radius=1, shapeName=tongue_04_sphere, center=<0 0.5 0>}]
}
constraints [
 {bone1Name=R_lip_corner, bone2Name=head, swing1LimitDegrees=10, swing2LimitDegrees=20, twistLimitDegrees=2}
 {bone1Name=L_lip_corner, bone2Name=head, swing1LimitDegrees=10, swing2LimitDegrees=20, twistLimitDegrees=2}
 {bone1Name=L_cheek_outer, bone2Name=head, swing1LimitDegrees=15, swing2LimitDegrees=15, twistLimitDegrees=2}
 {bone1Name=R_cheek_outer, bone2Name=head, swing1LimitDegrees=15, swing2LimitDegrees=15, twistLimitDegrees=2}
 {bone1Name=tongue, bone2Name=head, swing1LimitDegrees=5, swing2LimitDegrees=5, twistLimitDegrees=2}
 {bone1Name=tongue_02, bone2Name=tongue, swing1LimitDegrees=3, swing2LimitDegrees=3, twistLimitDegrees=1}
 {bone1Name=tongue_03, bone2Name=tongue_02, swing1LimitDegrees=2, swing2LimitDegrees=5, twistLimitDegrees=2}
 {bone1Name=tongue_04, bone2Name=tongue_03, swing1LimitDegrees=2, swing2LimitDegrees=5, twistLimitDegrees=2}
 {
  bone1Name=C_lip_lower_mid
  bone2Name=head
  swing1LimitDegrees=45
  swing1Motion=Free
  swing2LimitDegrees=45
  swing2Motion=Free
  twistLimitDegrees=45
  twistMotion=Free
 }
 {
  bone1Name=C_lip_upper_mid
  bone2Name=head
  swing1LimitDegrees=45
  swing1Motion=Free
  swing2LimitDegrees=45
  swing2Motion=Free
  twistLimitDegrees=45
  twistMotion=Free
 }
]
