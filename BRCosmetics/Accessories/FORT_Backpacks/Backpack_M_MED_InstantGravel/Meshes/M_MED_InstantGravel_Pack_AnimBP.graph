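# Animation Blueprint for the InstantGravel backpack cosmetic, an animated pack with jowls, tongue and paws.
# Front-end or in-game state machines play idles with timed or chest-opened reactions, then feed cheek AnimDynamics and RigidBody physics into a cached pose.
# A skydiving bool selects between jowl physics setups, the skydive one adding limb bone offsets.
# Opening a matching glider spawns transfer FX and hides the pack via material parameter and dispatcher; skydiving or closing unhides it.
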AnimBlueprint /BRCosmetics/Accessories/FORT_Backpacks/Backpack_M_MED_InstantGravel/Meshes/M_MED_InstantGravel_Pack_AnimBP
target_skeleton=&/M_MED_InstantGravel_Pack_Skeleton
graphs [
 AnimGraph {nodes [
   n1 = Anim Dynamics - Bone: R_lip_corner!(ComponentPose, Alpha=1, GravityOverride:Vector="0,0,0", ExternalForce="0,0,0") -> (Pose) {
    angularDampingOverride=0.9999
    angularSpringConstant=25
    bAngularSpring=true
    bGravityOverrideInSimSpace=true
    bOverrideAngularDamping=true
    bUseGravityOverride=true
    simulationSpace=BoneRelative
    alphaBoolBlend {blendOption=Linear}
    boundBone {boneName=R_lip_corner}
    componentAppliedLinearAccClamp=<3500 3500 3500>
    componentLinearAccScale=<1 0.8 1.3>
    componentLinearVelScale=<1 0.8 1.3>
    physicsBodyDefinitions [{
      sphereCollisionRadius=0
      boundBone {boneName=R_lip_corner}
      boxExtents=<20 20 10>
      constraintSetup {angularLimitsMax=<10 30 20>, angularLimitsMin=<-10 -30 -20>, angularTarget=<0 0 0>, linearAxesMax=<0.4 2 0.5>, linearAxesMin=<-0.4 -2 -0.5>}
      localJointOffset=<0 -2 2>
     }]
    relativeSpaceBone {boneName=root}
   }
   Anim Dynamics - Bone: L_lip_corner!(ComponentPose=@n1, Alpha=1, GravityOverride:Vector="0,0,0", ExternalForce="0,0,0") -> (Pose) {
    angularDampingOverride=0.9999
    angularSpringConstant=25
    bAngularSpring=true
    bGravityOverrideInSimSpace=true
    bOverrideAngularDamping=true
    bUseGravityOverride=true
    simulationSpace=BoneRelative
    alphaBoolBlend {blendOption=Linear}
    boundBone {boneName=L_lip_corner}
    componentAppliedLinearAccClamp=<3500 3500 3500>
    componentLinearAccScale=<1 0.8 1.3>
    componentLinearVelScale=<1 0.8 1.3>
    physicsBodyDefinitions [{
      sphereCollisionRadius=0
      boundBone {boneName=L_lip_corner}
      boxExtents=<20 20 10>
      constraintSetup {angularLimitsMax=<10 30 20>, angularLimitsMin=<-10 -30 -20>, angularTarget=<0 0 0>, linearAxesMax=<0.4 2 0.5>, linearAxesMin=<-0.4 -2 -0.5>}
      localJointOffset=<3 3 0>
     }]
    relativeSpaceBone {boneName=root}
   }
   Component To Local()
   n2 = Get Gravity Override()
   n3 = RigidBody!(
    ComponentPose=Local To Component(LocalPose=Use cached pose 'Pose'())
    AlphaCurveName:Name=SkyDive
    OverrideWorldGravity:Vector=@n2
    ExternalForce
   ) -> (Pose) {
    alphaInputType=Curve
    bOverrideWorldGravity=true
    overridePhysicsAsset=&/M_MED_InstantGravel_Pack_Physics_Jowls_Sky
    simulationSpace=BaseBoneSpace
    alphaBoolBlend {blendOption=Linear}
    alphaScaleBiasClamp {clampMax=0.8}
    baseBoneRef {boneName=root}
    componentAppliedLinearAccClamp=<6000 6000 6000>
    componentLinearAccScale=<0.5 1 1>
    componentLinearVelScale=<0.5 1 1>
    overrideWorldGravity {z=-980}
    simSpaceSettings {maxAngularAcceleration=2000, maxAngularVelocity=2000, maxLinearAcceleration=2000, maxLinearVelocity=2000}
   }
   n4 = "Transform (Modify) Bone - Bone: paw_upperarm_l"!(ComponentPose=@n3, Alpha=1, Rotation="20,0,0") -> (Pose) {
    rotationMode=BMM_Replace
    rotationSpace=BCS_BoneSpace
    translationSpace=BCS_BoneSpace
    alphaBoolBlend {blendOption=Linear}
    boneToModify {boneName=paw_upperarm_l}
    rotation {pitch=20}
    translation {y=-0.5}
   }
   n5 = "Transform (Modify) Bone - Bone: paw_upperarm_r"!(ComponentPose=@n4, Alpha=1, Rotation="20,0,0") -> (Pose) {
    rotationMode=BMM_Replace
    rotationSpace=BCS_BoneSpace
    translationSpace=BCS_BoneSpace
    alphaBoolBlend {blendOption=Linear}
    boneToModify {boneName=paw_upperarm_r}
    rotation {pitch=20}
    translation {y=-0.5}
   }
   n6 = "Transform (Modify) Bone - Bone: thight_l"!(ComponentPose=@n5, Alpha=1, Rotation="20,0,0") -> (Pose) {
    rotationMode=BMM_Replace
    rotationSpace=BCS_BoneSpace
    translationSpace=BCS_BoneSpace
    alphaBoolBlend {blendOption=Linear}
    boneToModify {boneName=thight_l}
    rotation {pitch=20}
    translation {y=-0.5}
   }
   n7 = "Transform (Modify) Bone - Bone: thight_r"!(ComponentPose=@n6, Alpha=1, Rotation="20,0,0") -> (Pose) {
    rotationMode=BMM_Replace
    rotationSpace=BCS_BoneSpace
    translationSpace=BCS_BoneSpace
    alphaBoolBlend {blendOption=Linear}
    boneToModify {boneName=thight_r}
    rotation {pitch=20}
    translation {y=-0.5}
   }
   n8 = RigidBody(ComponentPose=Local To Component(LocalPose=Use cached pose 'Pose'()), AlphaCurveName=SkyDive, OverrideWorldGravity=@n2) {
    alphaInputType=Curve
    bOverrideWorldGravity=true
    overridePhysicsAsset=&/M_MED_InstantGravel_Pack_Physics_Jowls
    simulationSpace=BaseBoneSpace
    alphaBoolBlend {blendOption=Linear}
    alphaScaleBiasClamp {bInterpResult=true, interpSpeedDecreasing=7, interpSpeedIncreasing=15}
    baseBoneRef {boneName=root}
    componentAppliedLinearAccClamp=<2000 2000 2000>
    componentLinearAccScale=<0.5 0.5 0.6>
    componentLinearVelScale=<0.5 0.5 0.6>
    overrideWorldGravity {z=-980}
    simSpaceSettings {maxAngularAcceleration=2000, maxAngularVelocity=2000, maxLinearAcceleration=2000, maxLinearVelocity=2000}
   }
   Output Pose(Result=Blend Poses by bool(BlendPose_0=Component To Local(ComponentPose=@n7), BlendPose_1=Component To Local(ComponentPose=@n8), bActiveValue=Get bIsSkyDiving()) {blendType=Linear}) {layerGroup=None}
   AnimGraphNode_StateMachine_0 = Idle + Reactive() {is=title AnimGraphNode_StateMachine}
   AnimGraphNode_StateMachine_1 = FE Idle + Reactive() {is=title AnimGraphNode_StateMachine}
   n9 = Get bIsInFrontEnd()
   n10 = "Transform (Modify) Bone - Bone: pack"!(
    ComponentPose=Local To Component(LocalPose=Blend Poses by bool(BlendPose_0=@AnimGraphNode_StateMachine_1, BlendPose_1=@AnimGraphNode_StateMachine_0, bActiveValue=@n9) {blendType=Linear})
    Alpha
    Translation="0,-0.5,0"
   ) -> (Pose) {
    translationMode=BMM_Replace
    translationSpace=BCS_BoneSpace
    alphaBoolBlend {blendOption=Linear}
    boneToModify {boneName=pack}
    translation {y=-0.5}
   }
   n11 = Anim Dynamics - Bone: R_cheek_outer!(ComponentPose=@n10, Alpha=0.3, GravityOverride:Vector="0,0,0", ExternalForce="0,0,0") -> (Pose) {
    angularDampingOverride=0.9999
    angularSpringConstant=25
    bAngularSpring=true
    bGravityOverrideInSimSpace=true
    bOverrideAngularDamping=true
    bUseGravityOverride=true
    simulationSpace=BoneRelative
    alphaBoolBlend {blendOption=Linear}
    boundBone {boneName=R_cheek_outer}
    componentAppliedLinearAccClamp=<3500 3500 3500>
    componentLinearAccScale=<1.2 2 1.5>
    componentLinearVelScale=<1.2 2 1.5>
    physicsBodyDefinitions [{
      sphereCollisionRadius=0
      boundBone {boneName=R_cheek_outer}
      boxExtents=<20 20 10>
      constraintSetup {angularLimitsMax=<5 5 5>, angularLimitsMin=<-5 -5 -5>, angularTarget=<0 0 0>, linearAxesMax=<0.2 0.4 0.3>, linearAxesMin=<-0.2 -0.4 -0.3>}
      localJointOffset=<0 -2 2>
     }]
    relativeSpaceBone {boneName=root}
   }
   n12 = Anim Dynamics - Bone: L_cheek_outer!(ComponentPose=@n11, Alpha=0.3, GravityOverride:Vector="0,0,0", ExternalForce="0,0,0") -> (Pose) {
    angularDampingOverride=0.9999
    angularSpringConstant=25
    bAngularSpring=true
    bGravityOverrideInSimSpace=true
    bOverrideAngularDamping=true
    bUseGravityOverride=true
    simulationSpace=BoneRelative
    alphaBoolBlend {blendOption=Linear}
    boundBone {boneName=L_cheek_outer}
    componentAppliedLinearAccClamp=<3500 3500 3500>
    componentLinearAccScale=<1.2 2 1.5>
    componentLinearVelScale=<1.2 2 1.5>
    physicsBodyDefinitions [{
      sphereCollisionRadius=0
      boundBone {boneName=L_cheek_outer}
      boxExtents=<20 20 10>
      constraintSetup {angularLimitsMax=<5 5 5>, angularLimitsMin=<-5 -5 -5>, angularTarget=<0 0 0>, linearAxesMax=<0.2 0.4 0.3>, linearAxesMin=<-0.2 -0.4 -0.3>}
      localJointOffset=<3 3 0>
     }]
    relativeSpaceBone {boneName=root}
   }
   n13 = Anim Dynamics - Bone: R_cheek_inner!(ComponentPose=@n12, Alpha=0.3, GravityOverride:Vector="0,0,0", ExternalForce="0,0,0") -> (Pose) {
    angularDampingOverride=0.9999
    angularSpringConstant=25
    bAngularSpring=true
    bGravityOverrideInSimSpace=true
    bOverrideAngularDamping=true
    bUseGravityOverride=true
    simulationSpace=BoneRelative
    alphaBoolBlend {blendOption=Linear}
    boundBone {boneName=R_cheek_inner}
    componentAppliedLinearAccClamp=<3500 3500 3500>
    componentLinearAccScale=<1 1.8 1.3>
    componentLinearVelScale=<1 1.8 1.3>
    physicsBodyDefinitions [{
      sphereCollisionRadius=0
      boundBone {boneName=R_cheek_inner}
      boxExtents=<20 20 10>
      constraintSetup {angularLimitsMax=<5 5 5>, angularLimitsMin=<-5 -5 -5>, angularTarget=<0 0 0>, linearAxesMax=<0.1 -0.2 0.2>, linearAxesMin=<-0.1 -0.2 -0.2>}
      localJointOffset=<0 -2 2>
     }]
    relativeSpaceBone {boneName=root}
   }
   n14 = Anim Dynamics - Bone: L_cheek_inner!(ComponentPose=@n13, Alpha=0.3, GravityOverride:Vector="0,0,0", ExternalForce="0,0,0") -> (Pose) {
    angularDampingOverride=0.9999
    angularSpringConstant=25
    bAngularSpring=true
    bGravityOverrideInSimSpace=true
    bOverrideAngularDamping=true
    bUseGravityOverride=true
    simulationSpace=BoneRelative
    alphaBoolBlend {blendOption=Linear}
    boundBone {boneName=L_cheek_inner}
    componentAppliedLinearAccClamp=<3500 3500 3500>
    componentLinearAccScale=<1 1.8 1.3>
    componentLinearVelScale=<1 1.8 1.3>
    physicsBodyDefinitions [{
      sphereCollisionRadius=0
      boundBone {boneName=L_cheek_inner}
      boxExtents=<20 20 10>
      constraintSetup {angularLimitsMax=<5 5 5>, angularLimitsMin=<-5 -5 -5>, angularTarget=<0 0 0>, linearAxesMax=<0.1 -0.2 0.2>, linearAxesMin=<-0.1 -0.2 -0.2>}
      localJointOffset=<3 3 0>
     }]
    relativeSpaceBone {boneName=root}
   }
   n15 = "Transform (Modify) Bone - Bone: tongue_03"!(ComponentPose=@n14, Alpha=1, Rotation="0,0,-2") -> (Pose) {
    rotationMode=BMM_Replace
    rotationSpace=BCS_BoneSpace
    translationSpace=BCS_BoneSpace
    alphaBoolBlend {blendOption=Linear}
    boneToModify {boneName=tongue_03}
    rotation {roll=-2}
    translation {y=-0.5}
   }
   n16 = RigidBody!(ComponentPose=@n15, Alpha:float=0.6, OverrideWorldGravity=@n2, ExternalForce) -> (Pose) {
    bOverrideWorldGravity=true
    overridePhysicsAsset=&/M_MED_InstantGravel_Pack_Physics
    simulationSpace=BaseBoneSpace
    alphaBoolBlend {blendOption=Linear}
    baseBoneRef {boneName=root}
    componentAppliedLinearAccClamp=<2000 2000 2000>
    componentLinearAccScale=<0.5 0.5 0.6>
    componentLinearVelScale=<0.5 0.5 0.6>
    overrideWorldGravity {z=-980}
    simSpaceSettings {maxAngularAcceleration=2000, maxAngularVelocity=2000, maxLinearAcceleration=2000, maxLinearVelocity=2000}
   }
   Save cached pose 'Pose'(Pose=Component To Local(ComponentPose=@n16)) {cacheName=Pose}
  ]}
 Idle + Reactive {
  entry=Idle
  path=AnimGraph.AnimGraphNode_StateMachine_0.Idle + Reactive
  ids=id
  nodes [
   AnimStateNode_0 = Idle()
   AnimStateNode_1 = Reactive() {is=title AnimStateNode}
   AnimStateTransitionNode_0 = AnimStateTransitionNode()
   AnimStateTransitionNode_1 = AnimStateTransitionNode() {bAutomaticRuleBasedOnSequencePlayerInState=true, crossfadeDuration=0.5}
  ]
  links [
   {
    dst=Reactive
    rule=AnimGraph.AnimGraphNode_StateMachine_0.Idle + Reactive.AnimStateTransitionNode_0.Transition
    src=Idle
    via=AnimStateTransitionNode_0
   }
   {
    dst=Idle
    rule=AnimGraph.AnimGraphNode_StateMachine_0.Idle + Reactive.AnimStateTransitionNode_1.Transition
    src=Reactive
    via=AnimStateTransitionNode_1
   }
  ]
 }
 Idle {
  path=AnimGraph.AnimGraphNode_StateMachine_0.Idle + Reactive.AnimStateNode_0.Idle
  nodes [
   n1 = Random Sequence Player!() -> (Pose:Pose Link Structure) {
    is=title AnimGraphNode_RandomPlayer
    entries [
     {
      chanceToPlay=0.8
      maxLoopCount=0
      maxPlayRate=1
      minLoopCount=0
      minPlayRate=1
      sequence=/BRCosmetics/Accessories/FORT_Backpacks/Backpack_M_MED_InstantGravel/Animation/InstantGravel_Idle
      blendIn {blendOption=Linear, blendTime=0.08, customCurve=null}
     }
     {
      chanceToPlay=0.2
      maxLoopCount=0
      maxPlayRate=1
      minLoopCount=0
      minPlayRate=1
      sequence=/BRCosmetics/Accessories/FORT_Backpacks/Backpack_M_MED_InstantGravel/Animation/InstantGravel_IdleBreak
      blendIn {blendOption=Linear, blendTime=0.08, customCurve=null}
     }
    ]
   }
   Output Animation Pose(Result=@n1)
  ]
 }
 Reactive {
  path=AnimGraph.AnimGraphNode_StateMachine_0.Idle + Reactive.AnimStateNode_1.Reactive
  nodes [
   n1 = Sequence Player 'InstantGravel_Reactive'() {bLoopAnimation=false, sequence=/BRCosmetics/Accessories/FORT_Backpacks/Backpack_M_MED_InstantGravel/Animation/InstantGravel_Reactive}
   Output Animation Pose(Result=@n1)
  ]
 }
 Transition {
  path=AnimGraph.AnimGraphNode_StateMachine_0.Idle + Reactive.AnimStateTransitionNode_0.Transition
  nodes [Result(bCanEnterTransition=Get ChestOpened() -> (ChestOpened:bool))]
 }
 Transition {path=AnimGraph.AnimGraphNode_StateMachine_0.Idle + Reactive.AnimStateTransitionNode_1.Transition, nodes [Result()]}
 FE Idle + Reactive {
  entry=Idle+Break
  path=AnimGraph.AnimGraphNode_StateMachine_1.FE Idle + Reactive
  ids=id
  nodes [
   AnimStateNode_0 = Idle+Break() {is=title AnimStateNode, stateEntered {guid=#1, notifyName=FE Reactive}}
   AnimStateNode_1 = ReactiveFE() {is=title AnimStateNode}
   AnimStateTransitionNode_0 = AnimStateTransitionNode() {crossfadeDuration=0.3}
   AnimStateTransitionNode_1 = AnimStateTransitionNode() {bAutomaticRuleBasedOnSequencePlayerInState=true, crossfadeDuration=0.1}
  ]
  links [
   {
    dst=ReactiveFE
    rule=AnimGraph.AnimGraphNode_StateMachine_1.FE Idle + Reactive.AnimStateTransitionNode_0.Transition
    src=Idle+Break
    via=AnimStateTransitionNode_0
   }
   {
    dst=Idle+Break
    rule=AnimGraph.AnimGraphNode_StateMachine_1.FE Idle + Reactive.AnimStateTransitionNode_1.Transition
    src=ReactiveFE
    via=AnimStateTransitionNode_1
   }
  ]
 }
 Idle+Break {
  path=AnimGraph.AnimGraphNode_StateMachine_1.FE Idle + Reactive.AnimStateNode_0.Idle+Break
  nodes [
   n1 = Random Sequence Player() {entries [
     {
      chanceToPlay=0.8
      maxLoopCount=0
      maxPlayRate=1
      minLoopCount=0
      minPlayRate=1
      sequence=&InstantGravel_Idle
      blendIn {blendOption=Linear, blendTime=0.1, customCurve=null}
     }
     {
      chanceToPlay=0.2
      maxLoopCount=0
      maxPlayRate=1
      minLoopCount=0
      minPlayRate=1
      sequence=&InstantGravel_IdleBreak
      blendIn {blendOption=Linear, blendTime=0.1, customCurve=null}
     }
    ]}
   Output Animation Pose(Result=@n1)
  ]
 }
 ReactiveFE {
  path=AnimGraph.AnimGraphNode_StateMachine_1.FE Idle + Reactive.AnimStateNode_1.ReactiveFE
  nodes [
   n1 = Sequence Player 'InstantGravel_Reactive'() {bLoopAnimation=false, sequence=&InstantGravel_Reactive}
   Output Animation Pose(Result=@n1)
  ]
 }
 Transition {
  path=AnimGraph.AnimGraphNode_StateMachine_1.FE Idle + Reactive.AnimStateTransitionNode_0.Transition
  nodes [Result(bCanEnterTransition=Get FEReactivity() -> (FEReactivity:bool))]
 }
 Transition {path=AnimGraph.AnimGraphNode_StateMachine_1.FE Idle + Reactive.AnimStateTransitionNode_1.Transition, nodes [Result()]}
 Backpack_Hide_Dispatcher {nodes [Backpack_Hide_Dispatcher() {is=title K2Node_FunctionEntry}]}
 Backpack_Unhide_Dispatcher {nodes [Backpack_Unhide_Dispatcher() {is=title K2Node_FunctionEntry}]}
 EventGraph {nodes [
   Event BlueprintUpdateAnimation()
   > n1 = Gravity Override State Function DEPRECATED(CharacterPartAnimInstance=Self-Reference(), GameTimeInSeconds=GetGameTimeInSeconds(), GravityOverrideParams=Get Gravity Override Params())
   > Set Gravity Override(Gravity Override=@n1)
   Event OnParachuteOpened()
   > n2 = Get Dynamic Material Instance(SkeletalMeshComponent=GetPartSkeletalMeshComponent())
   > n3 = Branch(Condition=GliderItemDefinitionHasAnyMetaTag!(Tags:Array of Gameplay Tag Structures=Make Array("[0]"=(TagName="Cosmetics.MetaTags.InstantGravel")), bRequiresOpened:bool=true) -> (ReturnValue:bool))
   > n4 = Sequence()
   Delay(execute=@n4.then_1, Duration=0.21)
   > SetScalarParameterValue(self=@n2, ParameterName=Hide Element 01, Value=1)
   > Call Backpack Hide Dispatcher!(execute:Exec, self:M MED Instant Gravel Pack Anim BP Object Reference) -> (then:Exec) {is=title K2Node_CallDelegate}
   SpawnSystemAttached!(
    execute=@n4.then_0
    SystemTemplate:Niagara System Object Reference=/BRCosmetics/Gadgets/Assets/VinderTech_GliderChute/Glider_InstantGravel/FX/NS_Glider_InstantGravel_Transfer
    AttachToComponent:Scene Component Object Reference=GetPartSkeletalMeshComponent()
    AttachPointName:Name=pack
    Location:Vector="0, 0, 0"
    Rotation:Rotator="0, 0, 0"
    LocationType:EAttachLocation Enum=KeepRelativeOffset
    bAutoDestroy:bool=true
    bAutoActivate:bool=true
    PoolingMethod:ENCPoolMethod Enum=None
    bPreCullCheck:bool=true
   ) -> (then, ReturnValue:Niagara Particle System Component Object Reference)
   SetScalarParameterValue(execute=[@n5.then, @n6.then, @n3.else], self=@n2, ParameterName=Hide Element 01, Value=0)
   > Call Backpack Unhide Dispatcher!(execute:Exec, self:M MED Instant Gravel Pack Anim BP Object Reference) -> (then:Exec) {is=title K2Node_CallDelegate}
   n5 = Event OnBeginSkydiving()
   Event BlueprintPostEvaluateAnimation() {is=title K2Node_Event}
   > Set ChestOpened(ChestOpened:bool=false) -> (Output_Get:bool)
   AnimNotify_FE Reactive() {is=title K2Node_Event}
   > Set FEReactivity(FEReactivity:bool=false) -> (Output_Get:bool)
   > RetriggerableDelay(Duration=10)
   > Set FEReactivity(FEReactivity=true)
   n6 = Event OnParachuteClosed()
   n7 = Event BlueprintProcessAnimInputEvent()
   > Branch(Condition=IsNamed(self=@n7.AnimInputEvent, InEventName=aie.ChestOpened))
   > Set ChestOpened(ChestOpened=true)
   TryGetPawnOwner()
  ]}
]
variables [
 Backpack_Hide_Dispatcher {default=()}
 Backpack_Unhide_Dispatcher {default=()}
 ChestOpened {type=boolean}
 FEReactivity {type=boolean}
 Gravity Override {type=Vector}
 Gravity Override Params {type=GravityOverrideParamsStruct}
]
